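; Creative device Blueprint for the ATK vehicle spawner (V2), layered on a parent class that is not shown here.
; It exposes a Supports Wraps user option through a getter and a setter that read and write the SupportsWraps variable directly.
; The construction script only calls the parent, so the file itself adds no spawning logic.
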(fn GetSupportsWraps
  (return (Variables|UserOptions|Basic|GetSupportsWraps)))

(fn SetSupportsWraps
  (params (Supports%20Wraps :type Boolean))
  (Variables|UserOptions|Basic|SetSupportsWraps Supports%20Wraps))

(fn ConstructionScript
  (|CallParentFunction))
